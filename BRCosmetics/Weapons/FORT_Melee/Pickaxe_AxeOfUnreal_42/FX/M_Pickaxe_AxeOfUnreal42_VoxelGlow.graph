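# Translucent, unlit, two-sided FX material for the voxel glow on the Axe of Unreal 42 pickaxe, usable on Niagara mesh particles.
# Particle color, a U-mapped color ramp and double-sided Fresnel drive emissive through MF_FX_EmissiveAdjustments.
# Opacity combines Fresnel, particle alpha and depth fade, which Low and Medium quality replace with a constant.
# Vertex offset combines camera and normal pushes, scaled by a time-scrolled sine along V.
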Material /BRCosmetics/Weapons/FORT_Melee/Pickaxe_AxeOfUnreal_42/FX/M_Pickaxe_AxeOfUnreal42_VoxelGlow
graph {
 properties {
  bUsedWithMeshDeformer=true
  bUsedWithNanite=true
  bUsedWithNiagaraMeshParticles=true
  bUsedWithSkeletalMesh=true
  blendMode=BLEND_Translucent
  pixelDepthOffsetMode=PDOM_Legacy
  previewMesh=/Game/Weapons/FORT_Melee/Pickaxe_Cube_Ninja_Male/FX/SM_Pickaxe_Cube_Ninja_Male_Impact_ShockWave_01
  shadingModel=MSM_Unlit
  twoSided=true
  usedShadingModels=Unlit
 }
 nodes [
  n1 = ParticleColor()
  n2 = TextureCoordinate()
  n3 = ComponentMask(None=ColorRamp!(None=ComponentMask(None=@n2) {r=true}) {colorCurve="<self>.ColorCurve"}) {b=true, g=true, r=true}
  n4 = MaterialFunctionCall(
   Brightness=ScalarParameter() {defaultValue=1, parameterName=Fresnel_Brightness}
   Exponent=ScalarParameter() {defaultValue=4, parameterName=Fresnel}
   Camera Vector=CameraVectorWS()
  ) {materialFunction=&MF_Fresnel_FX}
  MaterialFunctionCall(Emissive=Multiply(A=Multiply(A=@n1.RGB, B=@n3), B=@n4.Fresnel Double Sided)) {materialFunction=&MF_FX_EmissiveAdjustments}
  n5 = DepthFade(FadeDistance=ScalarParameter() {defaultValue=1, parameterName=FadeDistance}) {fadeDistanceDefault=1}
  n6 = Constant() {r=1}
  n7 = QualitySwitch(Default=@n5, Low=@n6, High=@n5, Medium=@n6, Epic=@n5)
  Multiply(A=Multiply(A=Multiply(A=@n4.Fresnel Double Sided, B=@n7), B=ScalarParameter() {defaultValue=1, parameterName=Opacity}), B=@n1.A)
  n8 = ScalarParameter() {defaultValue=-6, parameterName=CameraOffsetAmount}
  n9 = VertexNormalWS()
  n10 = ScalarParameter() {defaultValue=4, parameterName=VertexNormalAmount}
  n11 = Multiply(A=Time(), B=ScalarParameter() {defaultValue=0.25, parameterName=WPONoiseSpeed})
  n12 = ComponentMask(None=TextureCoordinate()) {g=true}
  n13 = Sine(None=Frac(None=Add(A=@n11, B=@n12)))
  Multiply(
   A=Add(
    A=MaterialFunctionCall(Offset Amount=@n8) {materialFunction=/Engine/Functions/Engine_MaterialFunctions02/WorldPositionOffset/CameraOffset}.Camera Offset
    B=Multiply(A=@n9, B=@n10)
   )
   B=Multiply(A=ConstantBiasScale(None=@n13), B=ScalarParameter() {defaultValue=1.5, parameterName=WPONoiseStrength})
  )
 ]
}
