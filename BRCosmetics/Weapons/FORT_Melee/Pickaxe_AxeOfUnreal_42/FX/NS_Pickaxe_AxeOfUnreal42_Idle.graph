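# Idle Niagara effect for the Axe of Unreal 42 pickaxe, emitting small glowing voxels that drift, shrink and fade.
# NE_Voxels spawns particles intermittently while System.ActivateIdle is true, placing them at hand-listed local points with colors picked from an array.
# Each renders as both mesh and sprite, with curves driving color, a glow color and a per-particle randomized material parameter.
# User.VariantID is exposed, and NE_Glow has no renderers.
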NiagaraSystem /BRCosmetics/Weapons/FORT_Melee/Pickaxe_AxeOfUnreal_42/FX/NS_Pickaxe_AxeOfUnreal42_Idle
dependencies {
 usedDataInterfaces [
  &NiagaraDataInterfaceArrayColor
  &NiagaraDataInterfaceArrayPosition
  &NiagaraDataInterfaceCurve
  &NiagaraDataInterfaceColorCurve
  /Script/Niagara.NiagaraDataInterfaceArrayBool
 ]
 usedDynamicInputs [/Niagara/DynamicInputs/Arrays/SelectBoolFromArray, &MakeIntFromFloat]
 usedModules [
  &SystemState
  &self:SystemSpawnScript.SystemScriptSource.SystemScriptGraph.NiagaraNodeAssignment_0.SetVariables_#1
 ]
}
emitters {
 NE_Glow {rendererClasses [], renderers []}
 NE_Voxels {
  emitterUpdateScript [
   &EmitterState()
   &SpawnRate(
    Spawn Probability=0.3
    SpawnRate=&MakeCustomFloatFromBool(Bool=$System.ActivateIdle, True Float=10, False Float=0)
    Use Spawn Probability=true
   )
  ]
  particleSpawnScript [
   /SetVariables(Particles.CustomRandom:NiagaraFloat=&RandomRangeFloat(Minimum=0, Maximum=1)) {isSetParameters=true}
   &InitializeParticle(
    Color=&SelectLinearColorFromArray(
     Array Sampling Mode=::NewEnumerator0
     Direct Array Index=?
     Color Selection Array=DataInterface() {propertyValues {ColorData [<rgba 2 1.417 0.4583 1>, <rgba 0.5134 1.5734 2 1>, <rgba 0.147 0.252 2 1>, <rgba 2 1.1614 0.2687 1>, <rgba 0.5166 0.1671 1 1>]}}
     Random Seed=0
    )
    Lifetime Max=1
    Lifetime Min=0.8
    Lifetime Mode=::NewEnumerator1
    Mesh Scale Mode=::NewEnumerator3
    Mesh Uniform Scale=0.2
    Position Offset=/Niagara/DynamicInputs/Arrays/SelectPositionFromArray(
     Array Sampling Mode:ENiagaraArraySamplingMode=::NewEnumerator0
     Direct Array Index:NiagaraInt32=?
     Interpolation:NiagaraFloat=?
     Direct Array Mode:ENiagara_WrapClamp=::NewEnumerator0
     Evaluation Type:ENiagaraRandomnessEvaluation=::NewEnumerator0
     Normalized Interpolation:NiagaraBool=true
     Recalculate Random Each Loop:NiagaraBool=false
     Position Array:NiagaraDataInterfaceArrayPosition=DataInterface() {propertyValues {
       GpuSyncMode=SyncCpuToGpu
       MaxElements=0
       PositionData [
        <0.4662 -16.4709 67.6129>
        <1.9167 17.4435 35.15>
        <2.2374 15.1101 65.9401>
        <2.3707 -1.0704 65.6551>
        <2.3055 19.6013 60.4408>
        <0.906 30.855 57.0429>
        <0.3436 -11.6235 63.4281>
        <2.0061 -3.1489 34.7597>
        <-0.5356 -3.5402 63.0559>
        <0.3765 26.1867 37.0886>
        <2.1296 22.9976 53.1719>
        <1.6604 -14.3846 51.7804>
        <0.3765 -20.0005 47.3892>
        <1.4844 -7.5932 57.2244>
        <0.7112 -7.0806 26.1136>
        <0.1455 -15.4534 44.1448>
        <1.2035 18.6221 75.9888>
        <0.6772 21.0584 26.2337>
        <1.656 -9.4956 50.4235>
        <1.5881 29.3174 49.442>
        <0.4716 30.6838 68.0422>
        <1.5608 -1.5831 78.7954>
        <0.3765 -13.9013 38.9137>
        <-0.076 22.5137 42.9434>
        <0.9093 -19.1453 53.6591>
        <0.9649 25.7957 64.1915>
        <-0.8225 -7.8163 41.3146>
        <1.172 31.4458 43.6975>
        <0.552 39.7496 51.6015>
        <1.0188 -13.5717 75.1105>
       ]
      }}
     Randomness Mode:ENiagaraRandomnessMode=::NewEnumerator0
     Fixed Random Seed:NiagaraBool=false
     Random Seed:NiagaraInt32=0
    )
    Position Offset Coordinate Space=::Local
    Sprite Size Mode=::NewEnumerator3
    Uniform Sprite Size=8
    UsePositionOffset=true
    Write Material Random=true
   )
   &InheritVelocity(Inherited Velocity Speed Limit=50, Speed Limit=true, UseSpeedThreshold=false)
   &AddVelocity(Velocity=<0 0 1>, Velocity Speed Scale=&RandomRangeFloat(Minimum=10, Maximum=20))
   &CameraOffset(Camera Offset Amount=10)
  ]
  particleUpdateScript [
   &Drag(Use Linear Drag=true, Use Rotational Drag=true)
   &SolveForcesAndVelocity(Clamp Velocity=false, Limit Acceleration=false)
   &ParticleState()
   &ScaleMeshSize(Scale Factor=&VectorFromFloat(Value=&FloatFromCurve(FloatCurve=DataInterface() {propertyValues {Curve {keys [
          <at 0.05 1>
          {arriveTangent=-1.0526, interpMode=RCIM_Cubic, leaveTangent=-1.0526, time=0.5, value=1}
          {arriveTangent=-0.974, tangentMode=RCTM_User, time=1, value=0}
         ]}}})))
   &UpdateMeshOrientation(Rotation Rate=&RandomRangeFloat(Minimum=0.2, Maximum=0.3), Rotation Vector=<1 1 1>)
   &ScaleColor(
    Scale Alpha=&FloatFromCurve(FloatCurve=DataInterface() {propertyValues {Curve {keys [{interpMode=RCIM_Cubic, tangentMode=RCTM_Break, time=0.3, value=1}, <at 1 0>]}}})
    Scale RGB=&MakeVectorFromLinearColorRGB(Color=&ColorFromCurve(ColorCurve=DataInterface() {propertyValues {
        AlphaCurve {keys [<at 0 1>]}
        BlueCurve {keys [<at 0.0062 5>,<at 0.5636 1.5>,<at 0.9882 1.2>]}
        GreenCurve {keys [<at 0.0062 5>,<at 0.5636 1.5>,<at 0.9882 1.2>]}
        RedCurve {keys [<at 0.0062 5>,<at 0.5636 1.5>,<at 0.9882 1.2>]}
       }}))
    ScaleA=true
    ScaleRGB=true
   )
   /SetVariables(Particles.GlowColor=&MakeLinearColorFromVectorAndFloat(
     "Vector (RGB)"=&MakeVectorFromLinearColorRGB(Color=$Particles.Color)
     "Float (Alpha)"=&FloatFromCurve(FloatCurve=DataInterface() {propertyValues {Curve {keys [<at 0 1>,<at 1 0>]}}}, Scale Curve=0.1)
    )) {isSetParameters=true}
   &DynamicMaterialParameters(
    Index 0 Param 1=&FloatFromCurve(
     FloatCurve=DataInterface() {propertyValues {Curve {defaultValue=1, keys [{tangentMode=RCTM_User, time=0, value=-1}, {tangentMode=RCTM_User, time=1, value=0}]}}}
     CurveIndex=$Particles.CustomRandom
    )
    Module.Index0Float0Write=true
    Module.Index0Float1Write=true
    Module.Index0Float2Write=true
    Module.Index0Float3Write=true
   )
  ]
  rendererClasses [&NiagaraMeshRendererProperties, &NiagaraSpriteRendererProperties]
  renderers [
   {rendererClass=&NiagaraMeshRendererProperties, rendererIndex=0}
   {rendererClass=&NiagaraSpriteRendererProperties, rendererIndex=1}
  ]
 }
}
summary {
 systemName=NS_Pickaxe_AxeOfUnreal42_Idle
 emitters [
  {emitterName=NE_Voxels, rendererClasses [&NiagaraMeshRendererProperties, &NiagaraSpriteRendererProperties]}
  {emitterName=NE_Glow, rendererClasses []}
 ]
 userVariables [User.VariantID {type=NiagaraFloat, defaultValue=0, description=""}]
}
