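# Niagara impact effect for the AxeOfUnreal 42 pickaxe, showing a colored flash with flying debris and smoke.
# NE_Flare packs staggered sprite, mesh and light layers into one emitter, using spawn groups and per-renderer attribute overrides.
# NE_Voxels and NE_VoxelsSmoke burst along System.DirectionVector with drag and gravity.
# System.IsVoxel switches their motion, lifetime or shading between styles, and front-end inputs rescale the effect for menu display.
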NiagaraSystem /BRCosmetics/Weapons/FORT_Melee/Pickaxe_AxeOfUnreal_42/FX/NS_Pickaxe_AxeOfUnreal42_Impact
dependencies {
 usedDataInterfaces [
  &NiagaraDataInterfaceArrayColor
  &NiagaraDataInterfaceCurve
  &NiagaraDataInterfaceColorCurve
  &NiagaraDataInterfaceVectorCurve
  &NiagaraDataInterfaceArrayFloat3
  &NiagaraDataInterfaceVector2DCurve
  /Script/Niagara.NiagaraDataInterfaceArrayBool
 ]
 usedDynamicInputs [
  /Niagara/DynamicInputs/Arrays/SelectBoolFromArray
  &MakeIntFromFloat
  /Niagara/DynamicInputs/Math/Clamp_Vector
  &FN_FrontEndVector3Modifier
 ]
 usedModules [
  &self:SystemSpawnScript.SystemScriptSource.SystemScriptGraph.NiagaraNodeAssignment_0.SetVariables_#1
  &self:SystemSpawnScript.SystemScriptSource.SystemScriptGraph.NiagaraNodeAssignment_2.SetVariables_#2
  &SystemState
  &self:SystemSpawnScript.SystemScriptSource.SystemScriptGraph.NiagaraNodeAssignment_1.SetVariables_#3
 ]
}
emitters {
 NE_Flare {
  emitterUpdateScript [
   /SetVariables(Transient.Shared.FloatFromCurve005_FloatCurve=DataInterface() {propertyValues {Curve {keys [
        {arriveTangent=-29.9725, interpMode=RCIM_Cubic, leaveTangent=-29.9725, tangentMode=RCTM_User, time=0, value=5}
        {interpMode=RCIM_Cubic, leaveTangent=-41.015, time=0.3, value=1}
       ]}}}) {isSetParameters=true}
   /SetVariables(Emitter.UV_Control:Vector4f=&MakeVector4(
     X=&FloatFromCurve(
      FloatCurve=?
      Transient.Shared.FloatFromCurve005_FloatCurve:NiagaraDataInterfaceCurve=DataInterface() {propertyValues {Curve {keys [
          {arriveTangent=-29.9725, interpMode=RCIM_Cubic, leaveTangent=-29.9725, tangentMode=RCTM_User, time=0, value=5}
          {interpMode=RCIM_Cubic, leaveTangent=-41.015, time=0.3, value=1}
         ]}}}
      CurveIndex=$Emitter.Age
     )
     Y=&FloatFromCurve(
      FloatCurve=?
      Transient.Shared.FloatFromCurve005_FloatCurve:NiagaraDataInterfaceCurve=DataInterface() {propertyValues {Curve {keys [
          {arriveTangent=-29.9725, interpMode=RCIM_Cubic, leaveTangent=-29.9725, tangentMode=RCTM_User, time=0, value=5}
          {interpMode=RCIM_Cubic, leaveTangent=-41.015, time=0.3, value=1}
         ]}}}
      CurveIndex=$Emitter.Age
     )
     Z=&Divide_Float(
      A=&Subtract_Float(
       A=&FloatFromCurve(
        FloatCurve=?
        Transient.Shared.FloatFromCurve005_FloatCurve:NiagaraDataInterfaceCurve=DataInterface() {propertyValues {Curve {keys [
            {arriveTangent=-29.9725, interpMode=RCIM_Cubic, leaveTangent=-29.9725, tangentMode=RCTM_User, time=0, value=5}
            {interpMode=RCIM_Cubic, leaveTangent=-41.015, time=0.3, value=1}
           ]}}}
        CurveIndex=$Emitter.Age
       )
       B=1
      )
      B=-2
     )
     W=&Divide_Float(
      A=&Subtract_Float(
       A=&FloatFromCurve(
        FloatCurve=?
        Transient.Shared.FloatFromCurve005_FloatCurve:NiagaraDataInterfaceCurve=DataInterface() {propertyValues {Curve {keys [
            {arriveTangent=-29.9725, interpMode=RCIM_Cubic, leaveTangent=-29.9725, tangentMode=RCTM_User, time=0, value=5}
            {interpMode=RCIM_Cubic, leaveTangent=-41.015, time=0.3, value=1}
           ]}}}
        CurveIndex=$Emitter.Age
       )
       B=1
      )
      B=-2
     )
    )) {isSetParameters=true}
   &EmitterState(Life Cycle Mode=::NewEnumerator1, Loop Behavior=::NewEnumerator1, Loop Delay=0.1, Loop Duration=1, UseLoopDelay=false)
   &SpawnBurst_Instantaneous(Spawn Count=1, Use Spawn Probability=false)
   &SpawnBurst_Instantaneous(Spawn Count=8, Spawn Group=1, Use Spawn Probability=false, UseLoopCountLimit=false) {module=SpawnBurst_Instantaneous001}
   &SpawnBurst_Instantaneous(Spawn Group=2, Use Spawn Probability=false, UseLoopCountLimit=false) {module=SpawnBurst_Instantaneous002}
   &SpawnBurst_Instantaneous(Spawn Count=1, Spawn Group=3, Use Spawn Probability=false) {module=SpawnBurst_Instantaneous003}
   &SpawnBurst_Instantaneous(Spawn Count=1, Spawn Group=4, Use Spawn Probability=false) {module=SpawnBurst_Instantaneous004}
  ]
  particleSpawnScript [
   /SetVariables(Particles.VisibilityTag=&SpawnGroup()) {isSetParameters=true}
   &InitializeParticle(
    Lifetime=&MergeHelper_FloatSwitch(
     0=0.3
     1=&RandomRangeFloat(Minimum=0.1, Maximum=0.15)
     2=0.15
     3=0.5
     4=0.45
     Index=&SpawnGroup()
     Num=::NewEnumerator3
    )
    Sprite Size Mode=::NewEnumerator3
    Uniform Sprite Size=&Multiply_Float(A=180, B=$System.FrontEndScale)
    UsePositionOffset=false
    Write Material Random=true
   )
   &DynamicMaterialParameters(
    Index 0 Param 1=&RandomRangeFloat(Minimum=0.4, Maximum=0.6)
    Index 0 Param 2=1
    Index 0 Param 3=1
    Index 0 Param 4=1
    Module.Param0WriteEnabled=true
    Module.Param1WriteEnabled=true
    Module.Param2WriteEnabled=true
    Module.Param3WriteEnabled=true
   )
   &CameraOffset(Camera Offset Amount=&MakeCustomFloatFromBool(Bool=$User.bIsFrontEndPreview, True Float=1, False Float=30))
   &InitialMeshOrientation(Use Rotation Vector=false)
   /SetVariables(Particles.FlashStarSubImageIndex:NiagaraFloat=&RandomRangeFloat(Minimum=0, Maximum=3)) {isSetParameters=true}
   /SetVariables(Particles.FlashStarSubImageIndexAlt:NiagaraFloat=&ModuloFloat(
     A=&Add_Float(A=$Particles.FlashStarSubImageIndex, B=&RandomRangeFloat(Minimum=0, Maximum=2))
     B=4
    )) {isSetParameters=true}
  ]
  particleUpdateScript [
   &ParticleState()
   &ScaleColor(
    Scale Alpha=&Multiply_Float(
     A=1
     B=&FloatFromCurve(FloatCurve=DataInterface() {propertyValues {Curve {keys [<at 0 0>,<at 0.2 1>,<at 0.6 0.65>,<at 1 0>]}}}, Scale Curve=0.3)
    )
    Scale RGB=&MakeVectorFromLinearColorRGB(Color=&ColorFromCurve(ColorCurve=DataInterface() {propertyValues {
        AlphaCurve {keys []}
        BlueCurve {keys [<at 0 0.9635 RCIM_Cubic>,<at 0.3043 0.651>,<at 0.6768 0.9844>]}
        GreenCurve {keys [<at 0 0.1054 RCIM_Cubic>,<at 0.3043 0.1373>,<at 0.6768 0.3402>]}
        RedCurve {keys [<at 0 0.9575 RCIM_Cubic>,<at 0.3043 0.3193>,<at 0.6768 0.2512>]}
       }}))
    ScaleA=true
    ScaleRGB=true
   )
   &ScaleSpriteSize(Scale Factor=&Vector2DFromFloat(Value=&FloatFromCurve(FloatCurve=DataInterface() {propertyValues {Curve {keys [<at 0 1.5 RCIM_Cubic>,<at 0.8 0>]}}}, Scale Curve=$System.FrontEndScale)))
   &FN_Sprite_Attributes(
    Color=&MakeLinearColorFromVectorAndFloat(
     "Vector (RGB)"=&MakeVectorFromLinearColorRGB(Color=&MultiplyLinearColor(
       Linear Color=&ColorFromCurve(ColorCurve=DataInterface() {propertyValues {
          AlphaCurve {keys []}
          BlueCurve {keys [<at 0.2716 0.1423 RCIM_Cubic>,<at 0.3043 0.651>,<at 0.6768 0.9844>]}
          GreenCurve {keys [<at 0.2716 0.5677 RCIM_Cubic>,<at 0.3043 0.5053>,<at 0.6768 0.3402>]}
          RedCurve {keys [<at 0.2716 0.9635 RCIM_Cubic>,<at 0.3043 0.196>,<at 0.6768 0.2512>]}
         }})
       Scale Factor=&MakeVector4_FromFloat(Float=0.5)
      ))
     "Float (Alpha)"=&Multiply_Float(
      A=&Multiply_Float(A=1, B=1)
      B=&FloatFromCurve(FloatCurve=DataInterface() {propertyValues {Curve {keys [<at 0 0.2>,<at 0.2 1>,<at 0.7 0.4>,<at 1 0>]}}})
     )
    )
    Create Color Attribute=true
    Create SpriteSize Attribute=true
    Module Output:ENiagara_AttributeSamplingApplyOutput=::NewEnumerator1
    SpriteSize=&Multiply_Vector2D_ByFloat(Float=1, Vector2D=$Particles.SpriteSize)
   )
   &FN_Mesh_Attributes(
    Color=&MultiplyLinearColor(
     Linear Color=&SelectLinearColorFromArray(
      Array Sampling Mode=::NewEnumerator1
      Direct Array Index=&RandomRangeInt(Minimum=0, Maximum=2)
      Color Selection Array=DataInterface() {propertyValues {ColorData [<rgba 1 0.7707 0.2084 1>, <rgba 0.2158 0.759 1 1>, <rgba 0.0754 0.1573 1 1>]}}
      Random Seed=?
     )
     Scale Factor=&MakeVector4FromVectorAndFloat(
      "Vector (RGB)"=&VectorFromFloat(Value=&FloatFromCurve(FloatCurve=DataInterface() {propertyValues {Curve {keys [<at 0 1>,<at 1 0>]}}}))
      "Float (Alpha)"=1
     )
    )
    Create Color Attribute=true
    Create Dynamic Parameter Attribute=true
    Create Scale Attribute=true
    Index 3 Param 1=1
    Index 3 Param 2=1
    Index 3 Param 3:NiagaraFloat=&Add_Float(
     A=&Multiply_Float(A=$Engine.System.Age, B=-5)
     B=&RandomRangeFloat(Minimum=0, Maximum=1)
    )
    Index 3 Param 4=1
    Module Output=::NewEnumerator1
    Scale=&Multiply_VectorByFloat(
     Vector=&Multiply_Vector(
      A=&UniformRangedVector(Minimum=<1.25 0.3 1>, Maximum=<1.6 0.5 1>)
      B=&VectorFromCurve(VectorCurve=DataInterface() {propertyValues {
         bOptimizeLUT=false
         XCurve {keys [{arriveTangent=1.8635, interpMode=RCIM_Cubic, leaveTangent=1.8635, tangentMode=RCTM_User, time=-0.0048, value=0.4}, <at 0.9737 1>]}
         YCurve {keys [
           {arriveTangent=0.1613, interpMode=RCIM_Cubic, leaveTangent=0.1613, tangentMode=RCTM_User, time=0, value=1}
           {arriveTangent=-1.3776, interpMode=RCIM_Cubic, leaveTangent=-1.3776, tangentMode=RCTM_User, time=0.9944, value=0}
          ]}
         ZCurve {keys [<at 0 1 RCIM_Cubic>, {arriveTangent=-1.635, interpMode=RCIM_Cubic, leaveTangent=-1.635, tangentMode=RCTM_User, time=0.9961, value=0}]}
        }})
     )
     Float=$System.FrontEndScale
    )
    Write Parameter Index 0:ENiagara_DynamicMaterialParametersInputMode=::NewEnumerator0
    Write Parameter Index 3:ENiagara_DynamicMaterialParametersInputMode=::NewEnumerator1
   )
   &FN_Sprite_Attributes(
    Color=&ColorFromCurve(ColorCurve=DataInterface() {propertyValues {
       AlphaCurve {keys [<at 0 1 RCIM_Cubic>,<at 1 0 RCIM_Cubic>]}
       BlueCurve {keys [<at 0.4987 0.3762 RCIM_Cubic>,<at 0.547 1 RCIM_Cubic>]}
       GreenCurve {keys [<at 0.4987 1.35 RCIM_Cubic>,<at 0.547 0.9742 RCIM_Cubic>]}
       RedCurve {keys [<at 0.4987 2 RCIM_Cubic>,<at 0.547 0.3889 RCIM_Cubic>]}
      }})
    Create Color Attribute=true
    Create SpriteSize Attribute=true
    Module Output:ENiagara_AttributeSamplingApplyOutput=::NewEnumerator1
    SpriteSize=&Multiply_Vector2D_ByFloat(
     Float=&Multiply_Float(
      A=&Multiply_Float(
       A=&RandomRangeFloat(Minimum=100, Maximum=150)
       B=&FloatFromCurve(FloatCurve=DataInterface() {propertyValues {Curve {keys [{arriveTangent=1.9953, interpMode=RCIM_Cubic, leaveTangent=1.9953, tangentMode=RCTM_User, time=0, value=0.3}, <at 1 1 RCIM_Cubic>]}}})
      )
      B=$System.FrontEndScale
     )
     Vector2D=<0.5 2>
    )
   ) {module=FN_Sprite_Attributes001}
   &FN_Sprite_Attributes(
    Color=&MultiplyLinearColor(
     Linear Color=&ColorFromCurve(ColorCurve=DataInterface() {propertyValues {
        AlphaCurve {keys [<at 0 1>]}
        BlueCurve {keys [<at 0.2357 0.6145 RCIM_Cubic>,<at 0.588 0.2236>,<at 0.6475 1>,<at 0.894 0.9844>]}
        GreenCurve {keys [<at 0.2357 2.7057 RCIM_Cubic>,<at 0.588 0.6077>,<at 0.6475 0.7267>,<at 0.894 0.2043>]}
        RedCurve {keys [<at 0.2357 5 RCIM_Cubic>,<at 0.588 1>,<at 0.6475 0.2088>,<at 0.894 0.0865>]}
       }})
     Scale Factor=&MakeVector4FromVectorAndFloat(
      "Vector (RGB)"=<1 1 1>
      "Float (Alpha)"=&FloatFromCurve(FloatCurve=DataInterface() {propertyValues {Curve {keys [
           {arriveTangent=0.1189, interpMode=RCIM_Cubic, leaveTangent=0.1189, tangentMode=RCTM_User, time=0, value=1}
           {arriveTangent=-1.3216, interpMode=RCIM_Cubic, leaveTangent=-1.3216, tangentMode=RCTM_User, time=1, value=0}
          ]}}})
     )
    )
    Create Color Attribute=true
    Create Dynamic Material Parameter Attributes=true
    Create SpriteSize Attribute=true
    Index 0 Param 1=0.65
    Index 0 Param 2=1
    Index 0 Param 3=1
    Index 0 Param 4=10
    Module Output:ENiagara_AttributeSamplingApplyOutput=::NewEnumerator1
    Module.Index0Float0Write=true
    Module.Index0Float1Write=true
    Module.Index0Float2Write=true
    Module.Index0Float3Write=true
    SpriteSize=&Vector2DFromFloat(Value=&Multiply_Float(
      A=&Multiply_Float(
       A=130
       B=&Multiply_Float(
        A=&FloatFromCurve(FloatCurve=DataInterface() {propertyValues {Curve {keys [
             {arriveTangent=2.185, interpMode=RCIM_Cubic, leaveTangent=2.185, tangentMode=RCTM_User, time=0, value=0}
             {interpMode=RCIM_Cubic, tangentMode=RCTM_User, time=1, value=1}
            ]}}})
        B=1
       )
      )
      B=$System.FrontEndScale
     ))
   ) {module=FN_Sprite_Attributes002}
   &FN_Mesh_Attributes(
    Color=&ColorFromCurve(ColorCurve=DataInterface() {propertyValues {
       AlphaCurve {keys [<at 0 1 RCIM_Cubic>,<at 1 0 RCIM_Cubic>]}
       BlueCurve {keys [
         <at 0.3037 0.6558 RCIM_Cubic>
         {arriveTangent=-1.9616, interpMode=RCIM_Cubic, leaveTangent=-1.9616, time=0.3231, value=0.2615}
         <at 0.5728 0.128>
        ]}
       GreenCurve {keys [
         <at 0.3037 1.6234 RCIM_Cubic>
         {arriveTangent=-3.2447, interpMode=RCIM_Cubic, leaveTangent=-3.2447, time=0.3231, value=0.9287}
         <at 0.5728 0.7503>
        ]}
       RedCurve {keys [<at 0.3037 2 RCIM_Cubic>, {arriveTangent=-3.7743, interpMode=RCIM_Cubic, leaveTangent=-3.7743, time=0.3231, value=2}, <at 0.5728 0.9844>]}
      }})
    Create Color Attribute=true
    Create Dynamic Parameter Attribute=true
    Create Orientation Attribute=true
    Create Scale Attribute=true
    Index 0 Param 2:NiagaraFloat=&FloatFromCurve(FloatCurve=DataInterface() {propertyValues {Curve {defaultValue=1, keys [{arriveTangent=2.6983, time=0, value=0.2}, {arriveTangent=2.2396, leaveTangent=2.2396, time=1, value=1.2}]}}})
    Module Output=::NewEnumerator1
    Module.Index0Float0Write=true
    Module.Index0Float1Write=true
    Module.Index0Float2Write=true
    Module.Index0Float3Write=true
    Orientation:Quat4f=/Game/Effects/Niagara/Modules/OrientMestToVectorFromCustomRot(
     InitialRotation:Quat4f=<0 0 0 1>
     Look At Direction:Vector3f=$System.DirectionVector
     Reference Axis Vector:Vector3f=<0 0 1>
     RotationVector:Vector3f=&SelectVectorFromArray(
      Array Sampling Mode=::NewEnumerator0
      Direct Array Index=?
      Vector Selection Array=DataInterface() {propertyValues {FloatData [<0 0 0>, <0 0 0.5>]}}
      Random Seed=0
     )
    )
    Scale=&Multiply_VectorByFloat(
     Vector=&Multiply_VectorByFloat(
      Vector=&VectorFromCurve(VectorCurve=DataInterface() {propertyValues {XCurve {keys [<at 0 1>,<at 1 1>]}, YCurve {keys [<at 0 1>,<at 1 1>]}, ZCurve {keys [<at 0 1>,<at 1 1>]}}})
      Float=&FloatFromCurve(FloatCurve=DataInterface() {propertyValues {Curve {keys [<at 0 0.2>,<at 1 0.5>]}}})
     )
     Float=$System.FrontEndScale
    )
   ) {module=FN_Mesh_Attributes001}
   &FN_Light_Attributes(
    Color=&ColorFromCurve(
     ColorCurve=DataInterface() {propertyValues {
       AlphaCurve {keys [<at 0 1>,<at 0.6808 1>]}
       BlueCurve {keys [<at 0.1313 0.2161>, <at 0.4826 0.1346>, {arriveTangent=-31.3347, time=0.5261, value=1}, <at 0.8532 1>]}
       GreenCurve {keys [<at 0.1313 0.8004>, <at 0.4826 0.4977>, {arriveTangent=-25.4984, time=0.5261, value=0.675}, <at 0.8532 0.3081>]}
       RedCurve {keys [<at 0.1313 2>, <at 0.4826 0.8238>, {arriveTangent=0.6412, time=0.5261, value=0.1665}, <at 0.8532 0.1496>]}
      }}
     CurveIndex=&Multiply_Float(A=$Particles.NormalizedAge, B=1)
    )
    Create Color Attribute=true
    Create Position Attribute=true
    Create Radius Attribute=true
    Position=<10 0 0>
    Radius=&FloatFromCurve(
     FloatCurve=DataInterface() {propertyValues {Curve {keys [{arriveTangent=6.4696, interpMode=RCIM_Cubic, leaveTangent=6.4696, tangentMode=RCTM_User, time=0, value=0.5}, <at 0.2 1>, <at 1 0>]}}}
     Scale Curve=80
    )
    Scale Color=&FloatFromCurve(
     FloatCurve=DataInterface() {propertyValues {Curve {keys [{arriveTangent=1.3205, interpMode=RCIM_Cubic, leaveTangent=1.3205, tangentMode=RCTM_User, time=0, value=1}, <at 1 0>]}}}
     CurveIndex=&Multiply_Float(A=$Particles.NormalizedAge, B=0.51)
     Scale Curve=5
    )
   )
   &FN_Sprite_Attributes(
    Color=&MultiplyLinearColor(
     Linear Color=&ColorFromCurve(ColorCurve=DataInterface() {propertyValues {
        AlphaCurve {keys [<at 0 0>,<at 0.3597 1>]}
        BlueCurve {keys [<at 0.4854 0.0784>,<at 0.9353 0.1504>]}
        GreenCurve {keys [<at 0.4854 0.8069>,<at 0.9353 0.3113>]}
        RedCurve {keys [<at 0.4854 1.5>,<at 0.9353 0.9844>]}
       }})
     Scale Factor=&MakeVector4FromVectorAndFloat(
      "Vector (RGB)"=<1 1 1>
      "Float (Alpha)"=&FloatFromCurve(
       FloatCurve=DataInterface() {propertyValues {Curve {keys [
           {arriveTangent=-2.3061, interpMode=RCIM_Cubic, leaveTangent=-2.3061, tangentMode=RCTM_User, time=0, value=1}
           {arriveTangent=-0.2646, interpMode=RCIM_Cubic, leaveTangent=-0.2646, tangentMode=RCTM_User, time=0.8, value=0}
          ]}}}
       Scale Curve=0.5
      )
     )
    )
    Create Color Attribute=true
    Create Dynamic Material Parameter Attributes=true
    Create SpriteSize Attribute=true
    Index 0 Param 1=0.65
    Index 0 Param 2=1
    Index 0 Param 3=1
    Index 0 Param 4=10
    Module Output:ENiagara_AttributeSamplingApplyOutput=::NewEnumerator1
    Module.Index0Float0Write=true
    Module.Index0Float1Write=true
    Module.Index0Float2Write=true
    Module.Index0Float3Write=true
    SpriteSize=&Vector2DFromFloat(Value=&Multiply_Float(
      A=&Multiply_Float(
       A=100
       B=&Lerp_Float(
        A=0.7
        B=1
        Alpha=&FloatFromCurve(FloatCurve=DataInterface() {propertyValues {Curve {keys [
             {arriveTangent=2.185, interpMode=RCIM_Cubic, leaveTangent=2.185, tangentMode=RCTM_User, time=0, value=0}
             {interpMode=RCIM_Cubic, tangentMode=RCTM_User, time=1, value=1}
            ]}}})
       )
      )
      B=$System.FrontEndScale
     ))
   ) {module=FN_Sprite_Attributes003}
   /Game/Sprite/FN_SpriteAlignmentAndFacing(SpriteAlignment:Vector3f=$System.DirectionVector, SpriteFacing:Vector3f=<1 0 0>)
   &FN_Sprite_Attributes(
    Alignment:Vector3f=&Multiply_VectorByFloat(Vector=$System.DirectionVector, Float=-1)
    Color=&ColorFromCurve(ColorCurve=DataInterface() {propertyValues {
       AlphaCurve {keys [<at 0.5632 1 RCIM_Cubic>,<at 1 0 RCIM_Cubic>]}
       BlueCurve {keys [<at 0.547 0.8333 RCIM_Cubic>,<at 0.5841 9.8113>]}
       GreenCurve {keys [<at 0.547 7.1732 RCIM_Cubic>,<at 0.5841 5.4414>]}
       RedCurve {keys [<at 0.547 10 RCIM_Cubic>,<at 0.5841 2.0288>]}
      }})
    Create Alignment Attribute=true
    Create Color Attribute=true
    Create Position Attribute=true
    Create SpriteSize Attribute=true
    Module Output:ENiagara_AttributeSamplingApplyOutput=::NewEnumerator1
    Position:NiagaraPosition=&AddVectorToPosition(Position=<0 0 0>, Vector=&Lerp_Vector(A=<0 0 0>, B=<1 1 1>, Alpha=0.5))
    SpriteSize=&Multiply_Vector2D(
     A=<20 200>
     B=&Vector2DFromCurve(Vector2Curve=DataInterface() {propertyValues {XCurve {keys [<at 0.5 1>,<at 1 0>]}, YCurve {keys [<at 0 1>,<at 1 1.5>]}}}, ScaleCurve=<1 1>)
    )
   ) {module=FN_Sprite_Attributes004}
  ]
  rendererClasses [
   &NiagaraSpriteRendererProperties
   &NiagaraMeshRendererProperties
   &NiagaraLightRendererProperties
  ]
  renderers [
   {rendererClass=&NiagaraSpriteRendererProperties}
   {rendererClass=&NiagaraSpriteRendererProperties, rendererIndex=1}
   {rendererClass=&NiagaraMeshRendererProperties, rendererIndex=2}
   {rendererClass=&NiagaraSpriteRendererProperties, rendererIndex=3}
   {rendererClass=&NiagaraSpriteRendererProperties, rendererIndex=4}
   {rendererClass=&NiagaraSpriteRendererProperties, rendererIndex=5}
   {rendererClass=&NiagaraMeshRendererProperties, rendererIndex=6}
   {rendererClass=&NiagaraLightRendererProperties, rendererIndex=7}
   {rendererClass=&NiagaraSpriteRendererProperties, rendererIndex=8}
   {rendererClass=&NiagaraSpriteRendererProperties, rendererIndex=9}
  ]
 }
 NE_Voxels {
  emitterUpdateScript [
   &EmitterState()
   &SpawnBurst_Instantaneous(
    Spawn Count=&RandomRangeInt(Minimum=8, Maximum=12)
    Use Spawn Probability=false
    UseLoopCountLimit=false
   )
  ]
  particleSpawnScript [
   &InitializeParticle(
    Color=&SelectLinearColorFromArray(
     Array Sampling Mode=::NewEnumerator0
     Direct Array Index=?
     Color Selection Array=DataInterface() {propertyValues {ColorData [<rgba 2 1.417 0.4583 1>, <rgba 0.5134 1.5734 2 1>, <rgba 0.147 0.252 2 1>, <rgba 2 1.1614 0.2687 1>]}}
     Random Seed=0
    )
    Lifetime=&MakeCustomFloatFromBool(
     Bool=$System.IsVoxel
     True Float=&RandomRangeFloat(Minimum=0.6, Maximum=0.9)
     False Float=&RandomRangeFloat(Minimum=0.2, Maximum=0.4)
    )
    Mesh Scale Mode=::NewEnumerator3
    Mesh Uniform Scale=&Multiply_Float(A=&RandomRangeFloat(Minimum=0.5, Maximum=0.7), B=$System.FrontEndScale)
    Sprite Size=&Multiply_Vector2D_ByFloat(
     Float=&Multiply_Float(A=&RandomRangeFloat(Minimum=8, Maximum=16), B=$System.FrontEndScale)
     Vector2D=<0.3 0.9>
    )
    Sprite Size Mode=::NewEnumerator4
    UsePositionOffset=false
    Write Material Random=true
   )
   &AddVelocity(
    Cone Axis=$System.DirectionVector
    Module.UseConeFalloff=false
    Module.UseVelDistribution=false
    Override Seed=false
    Velocity Mode=::NewEnumerator2
    Velocity Speed=&Multiply_Float(A=&RandomRangeFloat(Minimum=150, Maximum=750), B=$System.FrontEndScale)
   )
   &ShapeLocation(
    MaskSpawnGroup=false
    Sphere Distribution=::NewEnumerator3
    Sphere Radius=&Multiply_Float(A=15, B=$System.FrontEndScale)
    UseOwnerScale=true
   )
   &DynamicMaterialParameters(
    Index 0 Param 1=2
    Index 0 Param 2=20
    Index 0 Param 3=1
    Index 0 Param 4=1
    Module.Index0Float0Write=true
    Module.Index0Float1Write=true
    Module.Index0Float2Write=true
    Module.Index0Float3Write=true
   )
   &CameraOffset(Camera Offset Amount=25)
  ]
  particleUpdateScript [
   &Drag(
    Drag=&FloatFromCurve(
     FloatCurve=DataInterface() {propertyValues {Curve {keys [{interpMode=RCIM_Cubic, leaveTangent=3.6589, tangentMode=RCTM_Break, time=0, value=0.15}, <at 0.8 1>]}}}
     Scale Curve=8
    )
    Use Linear Drag=true
    Use Rotational Drag=false
   )
   &GravityForce(Gravity=&Multiply_VectorByFloat(
     Vector=&MakeCustomVectorFromBool(
      Bool=$System.IsVoxel
      True Vector=&RandomRangeVector(Minimum=<0 0 200>, Maximum=<0 0 400>)
      False Vector=<1 0 0>
     )
     Float=&FN_FrontEndFloatModifier(Gameplay=1, Front End=0.4)
    ))
   &SolveForcesAndVelocity(Clamp Velocity=false, Limit Acceleration=false)
   &ParticleState()
   &ScaleMeshSize(Scale Factor=&VectorFromFloat(Value=&FloatFromCurve(FloatCurve=DataInterface() {propertyValues {Curve {keys [
          {interpMode=RCIM_Cubic, leaveTangent=5.173, tangentMode=RCTM_Break, time=0, value=0.5}
          {arriveTangent=1.0043, interpMode=RCIM_Cubic, leaveTangent=-0.3119, tangentMode=RCTM_Break, time=0.2, value=1}
          {arriveTangent=-2.3616, tangentMode=RCTM_User, time=1, value=0}
         ]}}})))
   &ScaleColor(
    Scale Alpha=&FloatFromCurve(FloatCurve=DataInterface() {propertyValues {Curve {keys [{interpMode=RCIM_Cubic, tangentMode=RCTM_Break, time=0.3, value=1}, <at 1 0>]}}})
    ScaleA=true
    ScaleRGB=true
   )
   &FN_Sprite_Attributes(
    Color=&MultiplyLinearColor(Linear Color=$Particles.Color, Scale Factor=<1 1 1 0.05>)
    Create Color Attribute=true
    Create SpriteSize Attribute=true
    Module Output:ENiagara_AttributeSamplingApplyOutput=::NewEnumerator1
    SpriteSize=&Multiply_Vector2D(A=$Particles.SpriteSize, B=<5 3>)
   )
  ]
  rendererClasses [&NiagaraMeshRendererProperties, &NiagaraSpriteRendererProperties]
  renderers [
   {rendererClass=&NiagaraMeshRendererProperties, rendererIndex=0}
   {rendererClass=&NiagaraSpriteRendererProperties, rendererIndex=1}
   {rendererClass=&NiagaraSpriteRendererProperties, rendererIndex=2}
  ]
 }
 NE_VoxelsSmoke {
  emitterUpdateScript [
   &EmitterState()
   &SpawnBurst_Instantaneous(
    Spawn Count=&RandomRangeInt(Minimum=3, Maximum=4)
    Use Spawn Probability=false
    UseLoopCountLimit=false
   )
  ]
  particleSpawnScript [
   &InitializeParticle(
    Color=&SelectLinearColorFromArray(
     Array Sampling Mode=::NewEnumerator0
     Direct Array Index=?
     Color Selection Array=DataInterface() {propertyValues {ColorData [<rgba 2 1.417 0.4583 1>, <rgba 0.5134 1.5734 2 1>, <rgba 2 0.3858 1.8954 1>]}}
     Random Seed=0
    )
    Lifetime Max=0.8
    Lifetime Min=0.6
    Lifetime Mode=::NewEnumerator1
    Mesh Scale Mode=::NewEnumerator3
    Mesh Uniform Scale=&Multiply_Float(
     A=&RandomRangeFloat(Minimum=0.7, Maximum=0.8)
     B=&FN_FrontEndFloatModifier(Gameplay=1, Front End=0.55)
    )
    Sprite Rotation Mode=::NewEnumerator2
    Sprite Size Mode=::NewEnumerator3
    Uniform Sprite Size=&Multiply_Float(
     A=&FN_FrontEndFloatModifier(Gameplay=1, Front End=0.45)
     B=&RandomRangeFloat(Minimum=90, Maximum=140)
    )
    UsePositionOffset=false
    Write Material Random=true
   )
   &AddVelocity(
    Cone Axis=$System.DirectionVector
    Module.UseConeFalloff=false
    Module.UseVelDistribution=false
    Override Seed=false
    Velocity Mode=::NewEnumerator2
    Velocity Speed=&Multiply_Float(
     A=&RandomRangeFloat(Minimum=50, Maximum=200)
     B=&MakeCustomFloatFromBool(Bool=$System.IsVoxel, True Float=1, False Float=2)
    )
   )
   &ShapeLocation(
    MaskSpawnGroup=false
    Sphere Distribution=::NewEnumerator3
    Sphere Radius=&FN_FrontEndFloatModifier(Gameplay=15, Front End=8)
    UseOwnerScale=true
   )
   &DynamicMaterialParameters(
    Index 0 Param 1=&RandomRangeFloat(Minimum=0, Maximum=1)
    Index 0 Param 2=1
    Module.Index0Float0Write=true
    Module.Index0Float1Write=true
    Module.Index0Float2Write=true
    Module.Index0Float3Write=true
   )
   &InitialMeshOrientation(
    Mesh Orientation Mode=::NewEnumerator7
    Rotation=&SelectVectorFromArray(
     Array Sampling Mode=::NewEnumerator0
     Direct Array Index=?
     Vector Selection Array=DataInterface() {propertyValues {FloatData [<0 0 0>, <0.25 0 0>, <0.5 0 0>, <0.75 0 0>]}}
     Random Seed=0
    )
    Use Rotation Vector=true
   )
  ]
  particleUpdateScript [
   &Drag(
    Drag=&FloatFromCurve(
     FloatCurve=DataInterface() {propertyValues {Curve {keys [{interpMode=RCIM_Cubic, leaveTangent=3.6589, tangentMode=RCTM_Break, time=0, value=0.15}, <at 0.8 1>]}}}
     Scale Curve=8
    )
    Use Linear Drag=true
    Use Rotational Drag=false
   )
   &GravityForce(Gravity=&Multiply_VectorByFloat(
     Vector=&RandomRangeVector(Minimum=<0 0 50>, Maximum=<0 0 100>)
     Float=&FN_FrontEndFloatModifier(Gameplay=1, Front End=0.4)
    ))
   &SolveForcesAndVelocity(Clamp Velocity=false, Limit Acceleration=false)
   &ParticleState()
   &ScaleSpriteSize(Uniform Curve Sprite Scale=DataInterface() {propertyValues {Curve {keys [
        {arriveTangent=2.6035, interpMode=RCIM_Cubic, leaveTangent=2.6035, tangentMode=RCTM_User, time=0, value=0}
        {interpMode=RCIM_Cubic, tangentMode=RCTM_User, time=1, value=1}
       ]}}})
   &ScaleMeshSize(Scale Factor=&VectorFromFloat(Value=&FloatFromCurve(
      FloatCurve=DataInterface() {propertyValues {Curve {keys [
          {interpMode=RCIM_Cubic, leaveTangent=2.4759, tangentMode=RCTM_Break, time=0, value=0}
          {arriveTangent=-0.1759, tangentMode=RCTM_User, time=1, value=1}
         ]}}}
      Scale Curve=0.6
     )))
   &DynamicMaterialParameters(
    Index 0 Param 1=&MakeCustomFloatFromBool(
     Bool=$System.IsVoxel
     True Float=0
     False Float=&Add_Float(
      A=&RandomRangeFloat(Minimum=0, Maximum=1)
      B=&FloatFromCurve(
       FloatCurve=DataInterface() {propertyValues {Curve {keys [
           <at 0 0 RCIM_Cubic>
           {arriveTangent=1.5947, interpMode=RCIM_Cubic, leaveTangent=1.5947, tangentMode=RCTM_User, time=0.2557, value=0.5257}
           <at 1 1 RCIM_Cubic>
          ]}}}
       CurveIndex=$Particles.Age
       Scale Curve=0.25
      )
     )
    )
    Index 0 Param 2=&MakeCustomFloatFromBool(
     Bool=$System.IsVoxel
     True Float=&FloatFromCurve(FloatCurve=DataInterface() {propertyValues {Curve {keys [
          {arriveTangent=2.4806, interpMode=RCIM_Cubic, leaveTangent=2.4806, tangentMode=RCTM_User, time=0, value=0}
          {arriveTangent=1.5524, interpMode=RCIM_Cubic, leaveTangent=1.5524, tangentMode=RCTM_User, time=1, value=1}
         ]}}})
     False Float=0
    )
    Module.Index0Float0Write=true
    Module.Index0Float1Write=true
    Module.Index0Float2Write=true
    Module.Index0Float3Write=true
   ) {module=DynamicMaterialParameters001}
   &Color(Color=&MakeCustomLinearColorFromBool(
     Bool=$System.IsVoxel
     True Color=&ColorFromCurve(ColorCurve=DataInterface() {propertyValues {
        AlphaCurve {keys [<at 0 1 RCIM_Cubic>,<at 1 0 RCIM_Cubic>]}
        BlueCurve {keys [<at 0.4165 0.7 RCIM_Cubic>,<at 0.4402 0.4 RCIM_Cubic>]}
        GreenCurve {keys [<at 0.4165 0.7 RCIM_Cubic>,<at 0.4402 0.4 RCIM_Cubic>]}
        RedCurve {keys [<at 0.4165 0.7 RCIM_Cubic>,<at 0.4402 0.4 RCIM_Cubic>]}
       }})
     False Color=&ColorFromCurve(ColorCurve=DataInterface() {propertyValues {
        AlphaCurve {keys [<at 0 0.4 RCIM_Cubic>,<at 1 0 RCIM_Cubic>]}
        BlueCurve {keys [<at 0 0.3513 RCIM_Cubic>,<at 0.5473 0.4 RCIM_Cubic>]}
        GreenCurve {keys [<at 0 0.5749 RCIM_Cubic>,<at 0.5473 0.4 RCIM_Cubic>]}
        RedCurve {keys [<at 0 0.7 RCIM_Cubic>,<at 0.5473 0.4 RCIM_Cubic>]}
       }})
    ))
  ]
  rendererClasses [&NiagaraMeshRendererProperties, &NiagaraSpriteRendererProperties]
  renderers [
   {rendererClass=&NiagaraMeshRendererProperties, rendererIndex=0}
   {rendererClass=&NiagaraSpriteRendererProperties, rendererIndex=1}
   {rendererClass=&NiagaraSpriteRendererProperties, rendererIndex=2}
  ]
 }
}
summary {
 systemName=NS_Pickaxe_AxeOfUnreal42_Impact
 emitters [
  {emitterName=NE_Voxels, rendererClasses [&NiagaraMeshRendererProperties, &NiagaraSpriteRendererProperties]}
  {
   emitterName=NE_Flare
   rendererClasses [
    &NiagaraSpriteRendererProperties
    &NiagaraMeshRendererProperties
    &NiagaraLightRendererProperties
   ]
  }
  {emitterName=NE_VoxelsSmoke, rendererClasses [&NiagaraMeshRendererProperties, &NiagaraSpriteRendererProperties]}
 ]
 userVariables [
  User.bIsFrontEnd {type=NiagaraBool}
  User.bIsFrontEndPreview {type=NiagaraBool}
  User.SwingSectionIndex {type=NiagaraInt32}
  User.VariantID {type=NiagaraFloat, defaultValue=0, description=""}
 ]
}
